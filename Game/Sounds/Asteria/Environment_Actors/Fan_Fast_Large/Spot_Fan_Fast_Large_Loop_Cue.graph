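# Looping spot sound cue for a large, fast environment fan.
# Two identical stacks blend a blimp jump-pad loop with air-duct and BuildaBrella fan loops, both fan loops pitched up by fixed modulators.
# One stack uses no-focus attenuation and the indoor ambient class, the other occluded attenuation and the gadgets class, mixed slightly louder.
# Overridden concurrency tightly caps simultaneous instances.
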SoundCue /Game/Sounds/Asteria/Environment_Actors/Fan_Fast_Large/Spot_Fan_Fast_Large_Loop_Cue
root=Mixer_2
nodes [
 {
  id=Attenuation_0
  children [Mixer_0]
  properties {attenuationSettings=/Game/Sounds/Artemis/Environment_Actors/AirborneFans/Spot_AirborneFan_Blades_Occluded_Close_Att}
 }
 {
  id=Attenuation_1
  children [Mixer_1]
  properties {attenuationSettings=/Game/Sounds/Artemis/Environment_Actors/AirborneFans/Spot_AirborneFan_Blades_NoFocus_Close_Att}
 }
 {id=Mixer_0, children [WavePlayer_0,Modulator_0,Modulator_1], properties {inputVolume [1,0.25,1]}}
 {id=Mixer_1, children [WavePlayer_3,Modulator_2,Modulator_3], properties {inputVolume [1,0.25,1]}}
 {id=Mixer_2, children [SoundClass_1,SoundClass_0], properties {inputVolume [0.7,1]}}
 {id=Modulator_0, children [WavePlayer_1], properties {pitchMax=1.2, pitchMin=1.2}}
 {id=Modulator_1, children [WavePlayer_2], properties {pitchMax=1.2, pitchMin=1.2}}
 {id=Modulator_2, children [WavePlayer_4], properties {pitchMax=1.2, pitchMin=1.2}}
 {id=Modulator_3, children [WavePlayer_5], properties {pitchMax=1.2, pitchMin=1.2}}
 {id=SoundClass_0, children [Attenuation_0], properties {soundClassOverride=&Fort_Gadgets}}
 {id=SoundClass_1, children [Attenuation_1], properties {soundClassOverride=&Fort_Ambient_Indoors_Athena}}
 {id=WavePlayer_0, properties {bLooping=true, soundWaveAssetPtr=/Game/Sounds/Artemis/POI/Blimps/Spots/Blimp_Jump_Pad_Ambient_Loop_Mono}}
 {id=WavePlayer_1, properties {bLooping=true, soundWaveAssetPtr=/Game/Sounds/Apollo/Props/AirDucts/AMB_Air_Duct_Fan_Loop_01}}
 {id=WavePlayer_2, properties {bLooping=true, soundWaveAssetPtr=/Game/Sounds/Fort_UI_Sounds/S13_Frontend/S13_BuildaBrella_Fan_Loop}}
 {id=WavePlayer_3, properties {bLooping=true, soundWaveAssetPtr=&Blimp_Jump_Pad_Ambient_Loop_Mono}}
 {id=WavePlayer_4, properties {bLooping=true, soundWaveAssetPtr=&AMB_Air_Duct_Fan_Loop_01}}
 {id=WavePlayer_5, properties {bLooping=true, soundWaveAssetPtr=&S13_BuildaBrella_Fan_Loop}}
]
properties {
 bEnableSubmixSends=false
 bOverrideConcurrency=true
 maxDistance=1500
 volumeMultiplier=0.501
 concurrencyOverrides {maxCount=2, voiceStealReleaseTime=2}
 concurrencySet [/Game/Sounds/Concurrency_Rules/Athena_Prop]
 preEffectBusSends [{sendLevel=0.6}]
 soundSubmixSends [{minSendLevel=0, sendLevel=0, soundSubmix=&Fortnite_Reverb_Distant}]
}
